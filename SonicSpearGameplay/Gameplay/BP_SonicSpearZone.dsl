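; BP_SonicSpearZone is the Sonic Spear gameplay zone, applying a gameplay effect to overlapping actors and removing it by granted tag when they leave.
; On BeginPlay the authority spawns a separate cosmetic BP_SonicSpearZoneActor for the cloud and sets its radius to fire the notify.
; Locally controlled players entering trigger a rate-limited alert event.
; EndPlay broadcasts OnZoneEnded and increments the cosmetic actor's Zone Ending counter.
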(event EventBeginPlay
  ;[0 1 0.0425] Begin Play ( )
  (|CallParentFunction)
  (Networking|SwitchHasAuthority
    (:Authority
      (CallFunction|SpawnZoneActor)
      (Variables|Default|SetbCloudIsActive true))))

(event EventEndPlay
  (params (EndPlayReason :type EEndPlayReason%20Enum :default "Destroyed"))
  ;[0.15 0.1005 0 0.5] End Play
  (|CallParentFunction EndPlayReason)
  (Default|CallOnZoneEnded) ; the zone may not always be in relevancy, so be cautious binding to this
  (Variables|Default|SetbCloudIsActive false)
  (bind _zoneactor (Variables|Default|GetZoneActor)
    (:then
      (Variables|/SonicSpearGameplay/Gameplay/BP_SonicSpearZoneActor.BP_SonicSpearZoneActor_C|SetZone%20Ending (+ (Variables|/SonicSpearGameplay/Gameplay/BP_SonicSpearZoneActor.BP_SonicSpearZoneActor_C|GetZone%20Ending _zoneactor) 1) _zoneactor))
    (:else)))

(event OnComponentBeginOverlap(OverlapComponent)
  (params (OverlappedComponent :type Primitive%20Component%20Object%20Reference) (OtherActor :type Actor%20Object%20Reference) (OtherComp :type Primitive%20Component%20Object%20Reference) (OtherBodyIndex :type Integer) (bFromSweep :type Boolean) (SweepResult :type Hit%20Result%20Structure%20%28by%20ref%29))
  (bind _returnvalue (Utilities|Time|GetRealTimeSeconds))
  ;[0 0.1403 0.15 0.5] Add Inside Zone GE
  (Environment|ApplyGlobalEnvironmentGameplayEffecttoActor OtherActor (Variables|Default|GetInside%20Zone%20GE))
  (Networking|CanExecuteCosmeticEvents
    (:True
      (bind _asfort_player_pawn_athena (Utilities|Casting|CastToFortPlayerPawnAthena OtherActor))
      (if (Pawn|IsLocalController (Pawn|GetController _asfort_player_pawn_athena))
        (if (> (- _returnvalue (Variables|Default|GetLastAlertTime)) (ScalableFloat|GetValueatLevel (Variables|Default|GetMinTimeSecondsBetweenAlerts)))
          (GameplayEventRouter|BroadcastStatefulEvent (GameplayEventRouter|GetGlobalEventRouter) (Variables|Default|GetGameplayAlertEvent) 0)
          (Variables|Default|SetLastAlertTime _returnvalue))))) ; Am I a client?
)

(event OnComponentEndOverlap(OverlapComponent)
  (params (OverlappedComponent :type Primitive%20Component%20Object%20Reference) (OtherActor :type Actor%20Object%20Reference) (OtherComp :type Primitive%20Component%20Object%20Reference) (OtherBodyIndex :type Integer))
  (bind _returnvalue (Ability|GetFortAbilitySystemComponent OtherActor))
  (bind _remove_zone_ge_tags (Variables|Default|GetRemove%20Zone%20GE%20Tags))
  ;[0 0.04792 0.15 0.5] Remove Inside Zone GE
  (bind _returnvalue_2 (GameplayTags|HasAnyMatchingGameplayTags(Message) OtherActor _remove_zone_ge_tags))
  (if _returnvalue_2
    (Utilities|IsValid _returnvalue
      (:Is%20Valid
        (GameplayEffects|RemoveActiveEffectswithGrantedTags _returnvalue _remove_zone_ge_tags)))))

(event Custom|Spawn%20Zone%20Actor
  ;[0.15 0 0.1484 0.5] Spawn the cosmetics zone actor - this is the actor separate from the server-side functionality for cosmetics like the cloud fading in/fading out, etc.
  (bind _returnvalue (Game|SpawnActorfromClass "/SonicSpearGameplay/Gameplay/BP_SonicSpearZoneActor.BP_SonicSpearZoneActor_C" (Transformation|GetActorTransform) "Undefined" "MultiplyWithRoot" 0 self 0))
  (bind _output_get (Variables|Default|SetZoneActor _returnvalue))
  ;[0.000373 0.09854 0.15 0.5] manually set to fire notify on server
  (Variables|/SonicSpearGameplay/Gameplay/BP_SonicSpearZoneActor.BP_SonicSpearZoneActor_C|SetZone%20Radius (CallFunction|GetRadius) _output_get))

(fn ConstructionScript)
