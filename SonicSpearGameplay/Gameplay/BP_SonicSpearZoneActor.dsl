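; Sonic Spear zone actor that presents the zone's donut storm cloud, objective marker, radius-tracker circle, thunder and music.
; Replicated radius scales and fades in the cloud, then runs a size-selected EQS scan to set clamped cloud height above terrain.
; Server spawns the ping marker and binds zone end to cloud fade-out and self-destruction.
; Clients play thunder, and music gated by cloud state and AudioGameplayVolume presence.
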(event EventBeginPlay
  ;[0.01288 0.15 0.03422 0.5] Begin Play
  (|CallParentFunction)
  (CallFunction|ApplyCloudCosmeticHotfix)
  ;[0.4011 0.268 0.9661] Spawn an objective marker for more control on placement and to have the default "Marker" properties from a BGA class
  (Networking|SwitchHasAuthority
    (:Authority
      (if (ScalableFloat|GetValueasBool (Variables|Default|GetPing%20Enabled))
        (bind _returnvalue (Building|SpawnBuildingGameplayActor "/SonicSpearGameplay/Gameplay/Objective/BGA_SonicSpear_ObjectiveMarker.BGA_SonicSpear_ObjectiveMarker_C" (Transformation|GetActorTransform self) self true))
        (Variables|Default|SetObjective%20Marker _returnvalue)
        ;[1 1 1] cover race condition reference set, just in case
        (Variables|Default|GetDonut%20Cloud
          (:then
            _)
          (:else
            (bind _asbp_donut_cloud (Utilities|Casting|CastToBP_DonutCloud (Variables|/Script/Engine.ChildActorComponent|GetChildActor (Variables|Default|GetCloud%20Donut))))
            (Variables|Default|SetDonut%20Cloud _asbp_donut_cloud)))
        (Utilities|FlowControl|Sequence
          (:then_0
            ;[0 0.15 0.1397 0.5] set marker to ping when cloud finishes
            (Default|BindEventtoOnCloudFinishedForming (Variables|Default|GetDonut%20Cloud) (EventDispatchers|CreateEvent :!event Custom|On%20Cloud%20Finished%20Forming))
            ;[0.15 0.07746 0 0.5] Bind to destroy self when the zone is finished
            (bind _server_zone (Variables|Default|GetServer%20Zone)
              (:then
                (Default|BindEventtoOnZoneEnded _server_zone (EventDispatchers|CreateEvent :!event Custom|Fade%20Cloud%20Out)))
              (:else))))))
    (:Remote
      (Variables|Sound|SetCloudIsActive true)
      (Audio|SpawnSoundatLocation (Variables|Sound|GetIntroThunderStrikeSound) (Transformation|GetWorldLocation (Variables|Default|GetMusic))))))

(event Custom|PlayMusic
  (params (CloudIsActive? :type Boolean))
  (bind _returnvalue (GameInstanceSubsystems|GetMusicEventSubsystem))
  ;[0.15 0.03266 0.04484 0.5] Handle Music Playback
  (if (and CloudIsActive? (Variables|Sound|GetPlayerInsideAGV?))
    (bind _music (Variables|Default|GetMusic)
      (:then
        (if (not (Audio|Components|Audio|IsPlaying _music))
          (Components|Activation|Activate _music false)
          (Audio|AddEventbyComponent _returnvalue _music "(TagName=\"MusicEventSubsystem.Event.POI.Special\")" "(GameplayTags=)" 0 "")))
      (:else
        (Development|PrintString "Hello")))
    (else
      (bind _music_2 (Variables|Default|GetMusic)
        (:then
          (Audio|Components|Audio|FadeOut _music_2 4.0 0.0 "Linear" :!class /Script/Engine.AudioComponent)
          (Audio|RemoveEvent _returnvalue _music_2))
        (:else)))))

(event EventEndPlay
  (params (EndPlayReason :type EEndPlayReason%20Enum :default "Destroyed"))
  ;[0.1339 0.15 0 0.5] End Play
  (|CallParentFunction EndPlayReason)
  (CallFunction|ClearDispatcherBindings))

(event Custom|FadeOutMusic
  ;[0.15 0.03266 0.04484 0.5] Handle Music Playback
  (bind _music (Variables|Default|GetMusic)
    (:then
      (Audio|Components|Audio|FadeOut _music 4.0 0.0 "Linear" :!class /Script/Engine.AudioComponent)
      (Audio|RemoveEvent (GameInstanceSubsystems|GetMusicEventSubsystem) _music))
    (:else)))

(event OnToggledOn(AudioGameplayVolume)
  ;[0.15 0.03266 0.04484 0.5] Handle Music Playback
  ;;[0.05354 0.0398 0.15 0.5] Audio Gameplay Volume
  (bind _output_get (Variables|Sound|SetPlayerInsideAGV? true))
  (if (and _output_get (Variables|Sound|GetCloudIsActive))
    (bind _music (Variables|Default|GetMusic)
      (:then
        (if (Audio|Components|Audio|IsPlaying _music)
          (Audio|Parameter|ExecuteTriggerParameter (Variables|Default|GetMusic) "FadeIn")
          (else
            (CallFunction|PlayMusic (Variables|Sound|GetCloudIsActive)))))
      (:else))))

(event Custom|On%20Cloud%20Finished%20Forming
  ;[0 0.0405 0.15 0.5] Show spawn marker on clients
  (bind _objective_marker (Variables|Default|GetObjective%20Marker)
    (:then
      (bind _asbga_sonic_spear_objective_marker (Utilities|Casting|CastToBGA_SonicSpear_ObjectiveMarker _objective_marker))
      (Class|BGASonicSpearObjectiveMarker|CloudFinishedForming _asbga_sonic_spear_objective_marker))
    (:else)))

(event OnToggledOff(AudioGameplayVolume)
  ;[0.15 0.03266 0.04484 0.5] Handle Music Playback
  ;;[0.05354 0.0398 0.15 0.5] Audio Gameplay Volume
  (Variables|Sound|SetPlayerInsideAGV? false)
  (CallFunction|FadeOutMusic))

(event Custom|Fade%20Cloud%20Out
  ;[0.02998 0.15 0 0.5] Destroy self after cloud fading finishes
  (bind _donut_cloud (Variables|Default|GetDonut%20Cloud)
    (:then
      (Class|BPDonutCloud|FadeOutCloud _donut_cloud)
      (Default|BindEventtoOnCloudFinishedFading _donut_cloud (delegate Custom|Destroy%20Self))
      ;[0.15 0.006193 0.01694 0.5] Play Outro Thunder
      (Utilities|FlowControl|DoOnce false
        (:Completed
          (Networking|SwitchHasAuthority
            (:Remote
              (Audio|SpawnSoundatLocation (Variables|Sound|GetOutroThunderStrikeSound) (Transformation|GetWorldLocation (Variables|Default|GetMusic)))
              (CallFunction|FadeOutMusic))))))
    (:else)))

(event Custom|Destroy%20Self
  ;[0.02998 0.15 0 0.5] Destroy self after cloud fading finishes
  (Actor|DestroyActor))

(event Custom|On%20Landscape%20EQS%20finished
  (params (QueryInstance :type EQS%20Query%20Instance%20Object%20Reference) (QueryStatus :type EEnvQueryStatus%20Enum :default "Processing"))
  (CallFunction|SetDonutHeight QueryInstance QueryStatus))

(fn ApplyCloudCosmeticHotfix
  (if (ScalableFloat|GetValueasBool (Variables|Default|GetHide%20Donut%20Cloud))
    (Development|SetHiddeninGame (Variables|Default|GetCloud%20Donut) true true)))

(fn ClearDispatcherBindings
  (Utilities|FlowControl|Sequence
    (:then_0
      (bind _donut_cloud (Variables|Default|GetDonut%20Cloud)
        (:then
          (Default|UnbindEventfromOnCloudFinishedForming _donut_cloud (EventDispatchers|CreateEvent :!event On%20Cloud%20Finished%20Forming)))
        (:else)))
    (:then_1
      (bind _server_zone (Variables|Default|GetServer%20Zone)
        (:then
          (Default|UnbindEventfromOnZoneEnded _server_zone (EventDispatchers|CreateEvent :!event Fade%20Cloud%20Out)))
        (:else)))
    (:then_2
      (bind _donut_cloud_2 (Variables|Default|GetDonut%20Cloud)
        (:then
          (Default|UnbindEventfromOnCloudFinishedFading _donut_cloud_2 (EventDispatchers|CreateEvent :!event Destroy%20Self)))
        (:else)))))

(fn GetDeterminedDonutCloudHeight
  (params (Locations :type Array%20of%20Vectors) (Default%20Height :type Boolean))
  (if (not Default%20Height)
    (for _array_element (Variables|GetLocations)
      (bind (_array_element_x _array_element_y _array_element_z) (Math|Vector|BreakVector _array_element))
      (if (> _array_element_z (LocalVariables|GetCurrent%20Height))
        (LocalVariables|SetCurrent%20Height _array_element_z)))
    (return (+ (LocalVariables|GetCurrent%20Height) (ScalableFloat|GetValueatLevel (Variables|DynamicLandscapeCloud|GetBase%20Cloud%20Height%20to%20Add))))
    (else
      (return (ScalableFloat|GetValueatLevel (Variables|DynamicLandscapeCloud|GetDefault%20Cloud%20Height))))))

(fn OnRep_CloudIsActive
  (Networking|SwitchHasAuthority
    (:Remote
      (CallFunction|PlayMusic (Variables|Sound|GetCloudIsActive)))))

(fn OnRep_ZoneEnding
  (CallFunction|FadeCloudOut))

(fn OnRep_ZoneRadius
  (bind _asbp_donut_cloud (Utilities|Casting|CastToBP_DonutCloud (Variables|/Script/Engine.ChildActorComponent|GetChildActor (Variables|Default|GetCloud%20Donut))))
  (bind _output_get (Variables|Default|SetDonut%20Cloud _asbp_donut_cloud))
  (Transformation|SetRelativeScale3D (Variables|/WildEstate/TechArt/DonutCloud/Blueprints/BP_DonutCloud.BP_DonutCloud_C|GetSharedRoot _output_get) (Math|Vector|MakeVector (/ (Variables|Default|GetZone%20Radius) (Variables|Default|GetRadius%20Scale%20Divisor)) (/ (Variables|Default|GetZone%20Radius) (Variables|Default|GetRadius%20Scale%20Divisor)) (/ (Variables|Default|GetZone%20Radius) (Variables|Default|GetRadius%20Scale%20Divisor))))
  (Class|BPDonutCloud|FadeinCloud _output_get)
  (Utilities|FlowControl|Sequence
    (:then_0
      ;[0 0.15 0.1307 0.5] Choose which landscape EQS to run
      (switch Utilities|FlowControl|Switch|SwitchByRange (Variables|Default|GetZone%20Radius)
        (:Default
          _)
        (:MatchesRange_0
          (LocalVariables|SetLandscape%20EQS%20To%20Run (Variables|Default|GetSmall%20Landscape%20EQS)))
        (:MatchesRange_1
          _)
        (:MatchesRange_2
          (LocalVariables|SetLandscape%20EQS%20To%20Run (Variables|Default|GetLarge%20Landscape%20EQS))))
      (LocalVariables|SetLandscape%20EQS%20To%20Run (Variables|Default|GetMedium%20Landscape%20EQS)))
    (:then_1
      ;[0.15 0.1113 0 0.5] Run the landscape EQS scan
      (bind _returnvalue (AI|EQS|RunEQSQuery (LocalVariables|GetLandscape%20EQS%20To%20Run) self "AllMatching"))
      (AI|EQS|BindEventtoOnQueryFinished _returnvalue (EventDispatchers|CreateEvent :!event On%20Landscape%20EQS%20finished)))
    (:then_2
      (Networking|SwitchHasAuthority
        (:Authority
          (bind _returnvalue_2 (RadiusTracker|MakeRadiusTrackingInitializer 0.0 (Variables|Default|GetZone%20Radius) (Variables|Default|GetZone%20Radius) 2.0 0.0 (Variables|Default|GetRadiusTrackerCircleColor) (Variables|Default|GetCircle%20Drawing%20Material) self (Fortnite|GetAllFortPlayerControllers true) "Global"))
          (Fortnite|AddActorTrackingInfo (RadiusTracker|GetGameStateRadiusTrackerComponent) _returnvalue_2))))))

(fn SetDonutHeight
  (params (EQS%20Instance :type EQS%20Query%20Instance%20Object%20Reference) (Status :type EEnvQueryStatus%20Enum :default "Processing"))
  (Utilities|FlowControl|Sequence
    (:then_0
      (switch Utilities|FlowControl|Switch|SwitchonEEnvQueryStatus Status
        (:Success
          (bind (_resultlocations _returnvalue) (AI|EQS|GetQueryResultsasLocations EQS%20Instance))
          (LocalVariables|SetUse%20Default%20Height (not _returnvalue))
          (LocalVariables|SetLandscape%20Locations _resultlocations))
        (:Failed
          (LocalVariables|SetUse%20Default%20Height true))))
    (:then_1
      (bind _donut_cloud (Variables|Default|GetDonut%20Cloud)
        (:then
          (bind (_returnvalue_x _returnvalue_y) (Math|Vector|BreakVector (Transformation|GetActorLocation _donut_cloud)))
          (Transformation|SetActorLocation _donut_cloud (Math|Vector|MakeVector _returnvalue_x _returnvalue_y (Math|Float|Clamp(Float) (CallFunction|GetDeterminedDonutCloudHeight (LocalVariables|GetLandscape%20Locations) (select (ScalableFloat|GetValueasBool (Variables|DynamicLandscapeCloud|GetForce%20Default%20Cloud%20Height)) true (LocalVariables|GetUse%20Default%20Height))) (ScalableFloat|GetValueatLevel (Variables|DynamicLandscapeCloud|GetMin%20Cloud%20Height)) (ScalableFloat|GetValueatLevel (Variables|DynamicLandscapeCloud|GetMax%20Cloud%20Height)))) false true))
        (:else)))))

(fn ConstructionScript)
