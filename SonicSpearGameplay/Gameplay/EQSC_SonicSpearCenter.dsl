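; EQS context for the Sonic Spear gameplay that provides a single center location for environment queries.
; Its ProvideSingleLocation function returns the querier actor's world location, or the zero vector when that actor is not valid.
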(fn ProvideSingleLocation
  (params (QuerierObject :type Object%20Reference) (QuerierActor :type Actor%20Object%20Reference))
  (Utilities|IsValid QuerierActor
    (:Is%20Valid
      (return (Transformation|GetActorLocation QuerierActor)))
    (:Is%20Not%20Valid
      (return "0, 0, 0"))))
