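; EQS context for Sonic Spear that supplies the player pawns currently inside the querier's SonicSpearZone.
; It gathers Athena player pawns from the querier's overlapping actors, relying on the zone's capsule being set to overlap player pawns.
; A reverse loop removes pawns whose player state fails the Athena cast or is not a game participant, so indices stay valid during removal, then returns the remaining set.
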(fn ProvideActorsSet
  (params (QuerierObject :type Object%20Reference) (QuerierActor :type Actor%20Object%20Reference))
  ;[0 1 0 0.5] SonicSpearZone has a capsule that is set to overlap player pawns, so we can just query those results.
  ; However, we need to filter out non-participant pawns.
  (bind _output_get (LocalVariables|SetPawns (Collision|GetOverlappingActors QuerierActor "/Script/FortniteGame.FortPlayerPawnAthena" :!class /Script/Engine.Actor)))
  (bind (_arrayindex _arrayelement) (Utilities|Array|ReverseforEachLoop _output_get)
    (:LoopBody
      (bind _asfort_player_state_athena (Utilities|Casting|CastToFortPlayerStateAthena (Variables|/Script/Engine.Pawn|GetPlayerState _arrayelement))
        (:then
          (if (Class|FortPlayerStateAthena|IsAnAthenaGameParticipant _asfort_player_state_athena)
            _
            (else
              _)))
        (:CastFailed
          _))
      (Utilities|Array|RemoveIndex (LocalVariables|GetPawns) _arrayindex))
    (:Completed
      (return (LocalVariables|GetPawns)))))
