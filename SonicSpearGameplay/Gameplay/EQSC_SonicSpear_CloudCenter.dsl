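; EQS context that supplies a single location for Sonic Spear queries, named as the cloud center.
; It line traces vertically through the querier actor's position, from 5000 units above to 5000 below, on custom channel TraceTypeQuery10 with simple collision.
; It returns the hit location, projecting the querier onto whatever surface that channel finds, or the world origin when the trace misses or the querier is invalid.
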(fn ProvideSingleLocation
  (params (QuerierObject :type Object%20Reference) (QuerierActor :type Actor%20Object%20Reference))
  (bind (_returnvalue_x _returnvalue_y _returnvalue_z) (Math|Vector|BreakVector (Transformation|GetActorLocation QuerierActor)))
  (Utilities|IsValid QuerierActor
    (:Is%20Valid
      (bind (_outhit _returnvalue) (Collision|LineTraceByChannel :Start (Math|Vector|MakeVector _returnvalue_x _returnvalue_y (+ _returnvalue_z 5e+03)) :End (Math|Vector|MakeVector _returnvalue_x _returnvalue_y (- _returnvalue_z 5e+03)) :TraceChannel "TraceTypeQuery10" :bTraceComplex false :DrawDebugType "None" :bIgnoreSelf true :TraceColor "(R=1.000000,G=0.000000,B=0.000000,A=1.000000)" :TraceHitColor "(R=0.000000,G=1.000000,B=0.000000,A=1.000000)" :DrawTime 5.0))
      (if _returnvalue
        (return (.Location (Collision|BreakHitResult _outhit)))
        (else
          (return "0, 0, 0"))))
    (:Is%20Not%20Valid
      (return "0, 0, 0"))))
