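# EQS query that picks spawn locations for Sonic Spear objects around the Sonic Spear center context.
# It generates ringed points in a parameterized-radius cylinder, projected onto geometry by channel trace.
# Tests check safe zone, building interior and surface validity, and score water-body presence and proximity to players in zone.
# Two custom box-overlap tests use data-table gameplay tags, one rejecting tagged overlaps and the other scoring against them.
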EnvQuery /SonicSpearGameplay/Gameplay/EQS_SonicSpearObjectSpawn
options [{
  generator {
   name=FortQueryGenerator_PointsInCylinder_1
   class=FortQueryGenerator_PointsInCylinder
   properties {
    arcAngle=360
    bAutoSortTests=true
    bCanRunAsync=false
    bDefineArc=false
    bUseSpiralPattern=false
    center=&/EQSC_SonicSpearCenter.EQSC_SonicSpearCenter_C
    cylinderHeightMax=0
    cylinderHeightMin=0
    envQueryResultNormalizationOption=Default
    innerRadius=0
    navDataOverrideContext=null
    numberOfRings=10
    optionName=""
    pointOnRingSpacingMethod=ByNumberOfPoints
    pointsPerRing=20
    spaceBetweenPoints=50
    spaceBetweenSlices=500
    arcDirection {dirMode=TwoPoints, lineFrom=&EnvQueryContext_Querier, lineTo=null, rotation=&EnvQueryContext_Querier}
    outerRadius {
     dataBinding=&self:FortQueryGenerator_PointsInCylinder_1.AIDataProvider_QueryParams_0
     dataField=FloatValue
     defaultValue=1000
    }
    projectionData {
     bCanDisableTrace=true
     bCanProjectDown=true
     bCanTraceOnGeometry=true
     bCanTraceOnNavMesh=true
     bOnlyBlockingHits=true
     bTraceComplex=false
     extentX=0
     extentY=10
     extentZ=10
     navigationFilter=/Game/Athena/AI/NavFilters/NavFilter_NPC_AllNavmesh.NavFilter_NPC_AllNavmesh_C
     postProjectionVerticalOffset=0
     projectDown=30000
     projectUp=10000
     serializedChannel=ECC_GameTraceChannel11
     traceChannel=TraceTypeQuery10
     traceMode=GeometryByChannel
     traceProfileName=None
     traceShape=Line
     versionNum=1
    }
   }
  }
  tests [
   FortQueryTest_InsideAthenaSafeZone_0 {class=FortQueryTest_InsideAthenaSafeZone, properties {testPurpose=FilterAndScore}}
   EnvQueryTest_InsideWaterBody_0 {
    class=EnvQueryTest_InsideWaterBody
    properties {
     bDefineReferenceValue=false
     bIgnoreExclusionVolumes=false
     bIncludeWaves=false
     bSimpleWaves=true
     boolValue=true
     clampMaxType=None
     clampMinType=None
     filterType=Match
     floatValueMax=0
     floatValueMin=0
     multipleContextFilterOp=AllPass
     multipleContextScoreOp=AverageScore
     normalizationType=Absolute
     referenceValue=0
     scoreClampMax=0
     scoreClampMin=0
     scoringEquation=Constant
     scoringFactor=0.1
     testComment=""
     testPurpose=Score
    }
   }
   FortQueryTest_InsideBuilding_1 {class=FortQueryTest_InsideBuilding, properties {scoringFactor=0.5, testPurpose=FilterAndScore}}
   FortQueryTest_ValidSurface_0 {
    class=FortQueryTest_ValidSurface
    properties {
     flatSurfaceToleranceZ {
      dataBinding=&self:FortQueryTest_ValidSurface_0.FortAIDataProvider_ScalableFloat_0
      dataField=FloatValue
      defaultValue=10
     }
     radius {
      dataBinding=&self:FortQueryTest_ValidSurface_0.FortAIDataProvider_ScalableFloat_1
      dataField=FloatValue
      defaultValue=50
     }
     surfaceMaterials []
     traceOffsetDown {
      dataBinding=&self:FortQueryTest_ValidSurface_0.FortAIDataProvider_ScalableFloat_3
      dataField=FloatValue
      defaultValue=100
     }
     traceOffsetUp {
      dataBinding=&self:FortQueryTest_ValidSurface_0.FortAIDataProvider_ScalableFloat_2
      dataField=FloatValue
      defaultValue=100
     }
    }
   }
   EnvQueryTest_SonicSpearOverlapGameplayTags_1 {
    class=EnvQueryTest_SonicSpearOverlapGameplayTags
    properties {
     bDefineReferenceValue=false
     boolValue=false
     clampMaxType=None
     clampMinType=None
     filterType=Match
     floatValueMax=0
     floatValueMin=0
     multipleContextFilterOp=AllPass
     multipleContextScoreOp=AverageScore
     normalizationType=Absolute
     referenceValue=0
     scoreClampMax=0
     scoreClampMin=0
     scoringEquation=Constant
     scoringFactor=1
     testComment=""
     testPurpose=Filter
     matchAnyTags {dataTable=/SonicSpearGameplay/DataTables/SonicSpear_EQSOverlapGameplayTagsTable, rowName=Standard_FilterTags}
     overlapData {
      bOnlyBlockingHits=false
      bOverlapComplex=false
      bSkipOverlapQuerier=true
      extentX=500
      extentY=500
      extentZ=500
      overlapChannel=ECC_WorldStatic
      overlapShape=Box
      shapeOffset=<0 0 0>
     }
    }
   }
   EnvQueryTest_SonicSpearOverlapGameplayTags_2 {
    class=EnvQueryTest_SonicSpearOverlapGameplayTags
    properties {
     bDefineReferenceValue=false
     boolValue=false
     clampMaxType=None
     clampMinType=None
     filterType=Match
     floatValueMax=0
     floatValueMin=0
     multipleContextFilterOp=AllPass
     multipleContextScoreOp=AverageScore
     normalizationType=Absolute
     referenceValue=0
     scoreClampMax=0
     scoreClampMin=0
     scoringEquation=Constant
     scoringFactor=0.3
     testComment=""
     testPurpose=Score
     matchAnyTags {dataTable=&SonicSpear_EQSOverlapGameplayTagsTable, rowName=Standard_ScoreTags}
     overlapData {
      bOnlyBlockingHits=false
      bOverlapComplex=false
      bSkipOverlapQuerier=true
      extentX=500
      extentY=500
      extentZ=500
      overlapChannel=ECC_WorldStatic
      overlapShape=Box
      shapeOffset=<0 0 0>
     }
    }
   }
   EnvQueryTest_Distance_1 {
    class=EnvQueryTest_Distance
    properties {
     bDefineReferenceValue=true
     distanceTo=&/EQSC_SonicSpearPlayersInZone.EQSC_SonicSpearPlayersInZone_C
     filterType=Maximum
     floatValueMin=0
     multipleContextFilterOp=AnyPass
     scoringEquation=InverseLinear
     testComment=""
     testMode=Distance2D
     testPurpose=Score
     floatValueMax {
      dataBinding=&self:EnvQueryTest_Distance_1.AIDataProvider_QueryParams_0
      dataField=FloatValue
      defaultValue=5000
     }
     referenceValue {
      dataBinding=&self:EnvQueryTest_Distance_1.AIDataProvider_QueryParams_1
      dataField=FloatValue
      defaultValue=0
     }
     scoringFactor {
      dataBinding=&self:EnvQueryTest_Distance_1.FortAIDataProvider_ScalableFloat_0
      dataField=FloatValue
      defaultValue=0.5
     }
    }
   }
  ]
 }]
