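# Environment query for the Sonic Spear gameplay feature that samples points over a large area around the cloud center, apparently to find landscape height.
# A single donut generator builds concentric rings of points around the CloudCenter context.
# Each point is projected onto geometry by a long vertical trace on a custom game trace channel, without navmesh projection or offset.
# It has no tests, so points are neither filtered nor scored.
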EnvQuery /SonicSpearGameplay/Gameplay/EQS_SonicSpear_LandscapeHeight_Large
options [{
  generator {
   name=EnvQueryGenerator_Donut_0
   class=EnvQueryGenerator_Donut
   properties {
    bUseSpiralPattern=false
    center=&/EQSC_SonicSpear_CloudCenter.EQSC_SonicSpear_CloudCenter_C
    innerRadius=300
    numberOfRings=3
    outerRadius=12000
    pointsPerRing=10
    projectionData {
     extentX=0
     navigationFilter=null
     postProjectionVerticalOffset=0
     projectDown=30000
     projectUp=10000
     serializedChannel=ECC_GameTraceChannel11
     traceChannel=TraceTypeQuery10
     traceMode=GeometryByChannel
    }
   }
  }
  tests []
 }]
