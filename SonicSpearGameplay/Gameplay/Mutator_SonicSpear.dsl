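; Sonic Spear gameplay mutator that observes and logs when Sonic Spear zones begin and end.
; On BeginPlay it binds handlers to the OnZoneBegin and OnZoneEnd dispatchers, and on EndPlay it unbinds all events from both.
; Each handler receives the zone's data and writes its name, 2D location and radius to the Fortnite shipping log, with no other gameplay effect.
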(event EventBeginPlay
  ;[0 1 0 0.5] [BeginPlay]
  (EventDispatchers|BindEventtoOnZoneBegin (EventDispatchers|CreateEvent :!event Custom|HandleZoneBegin))
  (EventDispatchers|BindEventtoOnZoneEnd (EventDispatchers|CreateEvent :!event Custom|HandleZoneEnd)))

(event Custom|HandleZoneBegin
  (params (Zone_Location :type Vector%202D%20Structure) (Zone_Radius :type Float%20%28single-precision%29) (Zone_SafeZoneIndex :type Integer) (Zone_State :type ESonicSpearZoneState%20Enum) (Zone_Actor :type Sonic%20Spear%20Zone%20Object%20Reference) (Zone_Name :type Name))
  ;[1 1 0 0.5] [HandleZoneBegin]
  (Fortnite|FortShippingLog self (Utilities|String|Append "ZoneBegin: Name [" (Utilities|String|ToString(Name) Zone_Name) "] Loc [" (Utilities|String|ToString(Vector2d) Zone_Location) "] Radius [" (Utilities|String|ToString(Float) Zone_Radius) "]")))

(event Custom|HandleZoneEnd
  (params (Zone_Location :type Vector%202D%20Structure) (Zone_Radius :type Float%20%28single-precision%29) (Zone_SafeZoneIndex :type Integer) (Zone_State :type ESonicSpearZoneState%20Enum) (Zone_Actor :type Sonic%20Spear%20Zone%20Object%20Reference) (Zone_Name :type Name))
  ;[0.15 0.04485 0 0.5] [HandleZoneEnd]
  (Fortnite|FortShippingLog self (Utilities|String|Append "ZoneEnd: Name [" (Utilities|String|ToString(Name) Zone_Name) "] Loc [" (Utilities|String|ToString(Vector2d) Zone_Location) "] Radius [" (Utilities|String|ToString(Float) Zone_Radius) "]")))

(event EventEndPlay
  (params (EndPlayReason :type EEndPlayReason%20Enum :default "Destroyed"))
  ;[1 0 0 0.5] [EndPlay]
  (EventDispatchers|UnbindallEventsfromOnZoneBegin)
  (EventDispatchers|UnbindallEventsfromOnZoneEnd))

(fn ConstructionScript)
